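# Niagara system for the HydroBeam bass cosmetic, adding sloshing liquid, a water meter and drips to the instrument's skeletal mesh.
# Two active mesh emitters run spring-driven pendulums reacting to owner motion, passing orientation, speed, turbulence and a system timeline value to materials.
# A sprite emitter drips from two user-selected sockets, slowing as the owner moves faster.
# A front-end preview flag switches the main liquid's pivot; a duplicate liquid emitter is disabled.
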NiagaraSystem /SparksCosmetics/Instrument/Bass/HydroBeam/Effects/HyrdoBeam_Splash/NS_Bass_HydroBeam_Splash_Liquid1
dependencies {
 usedDataInterfaces [
  &NiagaraDataInterfaceCurve
  &NiagaraDataInterfaceArrayFloat3
  &NiagaraDataInterfaceArrayInt32
  &NiagaraDataInterfaceSkeletalMesh
  &NiagaraDataInterfaceArrayFloat
 ]
 usedDynamicInputs [&MakeFloatFromVector]
 usedModules [
  &self:SystemSpawnScript.SystemScriptSource.SystemScriptGraph.NiagaraNodeAssignment_0.SetVariables_#1
  &self:SystemSpawnScript.SystemScriptSource.SystemScriptGraph.NiagaraNodeAssignment_3.SetVariables_#2
  &SystemState
  &self:SystemSpawnScript.SystemScriptSource.SystemScriptGraph.NiagaraNodeAssignment_1.SetVariables_#3
 ]
}
emitters {
 Dripping001 {
  emitterUpdateScript [
   &EmitterState(Life Cycle Mode=::NewEnumerator1, Loop Delay=1, Loop Duration=4, UseLoopDelay=false)
   &SpawnRate(
    Spawn Probability=0.35
    SpawnRate=&Lerp_Float(
     A=3
     B=0
     Alpha=&RemapRange(
      Input Value To Remap=&VectorLength(Vector=$Engine.Owner.Velocity)
      Input Min=0
      Input Max=100
      Output Min=0
      Output Max=1
     )
    )
    Use Spawn Probability=true
   )
  ]
  particleSpawnScript [
   &InitializeParticle(
    Color=<rgba 1.6 1.8453 2 1>
    Lifetime=&RandomRangeFloat(Minimum=0.5, Maximum=1)
    Mass Mode=::NewEnumerator3
    Sprite Size=&Multiply_Vector2D_ByFloat(
     Float=&Multiply_Float(
      A=&MakeCustomFloatFromBool(Bool=$User.bIsFrontEndPreview, True Float=1, False Float=1)
      B=&RandomRangeFloat(Minimum=10, Maximum=15)
     )
     Vector2D=<0.1 0.5>
    )
    Sprite Size Mode=::NewEnumerator4
    UsePositionOffset=false
    Write Material Random=true
   )
   &AddVelocity(Velocity=&RandomRangeVector(Minimum=<0 0 -30>, Maximum=<0 0 -100>)) {enabled=false}
   &SkeletalMeshLocation(
    Bone / Socket Index=&SelectIntFromArray(
     Array Sampling Mode=::NewEnumerator1
     Direct Array Index=?
     Int Selection Array=DataInterface() {propertyValues {IntData [0,1,2]}}
     Random Seed=0
    )
    MaskSpawnGroup=false
    Mesh Sampling Type=::NewEnumerator5
    Orientation Sampling=::NewEnumerator1
    Position Sampling=::NewEnumerator2
    Skeletal Mesh=$System.Skeletal Mesh
    Socket Sampling Mode=::NewEnumerator5
   ) {enabled=false}
   &ShapeLocation(
    MaskSpawnGroup=false
    Non Uniform Scale=<0.75 0.45 1>
    Override Seed=false
    Shape Origin=$Particles.SkeletalMeshLocation.SampledPosition
    Sphere Radius=1.5
    UseOwnerScale=true
   ) {enabled=false}
   &ShapeLocation(
    Apply To Particle Position=::NewEnumerator1
    MaskSpawnGroup=false
    Non Uniform Scale=<0.75 0.2 0.5>
    Override Seed=false
    Shape Origin=&VectorFromFloat(Value=0)
    Sphere Radius=1
    UseOwnerScale=true
   ) {module=ShapeLocation001}
   &InheritVelocity(Inherited Velocity Amount Scale=<1 1 0>, Speed Limit=true, UseSpeedThreshold=false) {enabled=false}
   &PartitionParticles()
  ]
  particleUpdateScript [
   &ParticleState()
   /SetVariables(
    Particles.Socket001:NiagaraInt32=&MakeCustomIntFromBool(Bool=&SetBoolByFloatComparison(A=$User.dripsocket2, B=0), True INT=1, False INT=0)
    Particles.Socket_1:NiagaraInt32=&MakeCustomIntFromBool(Bool=&SetBoolByFloatComparison(A=$User.dripsocket1, B=0), True INT=1, False INT=0)
   ) {isSetParameters=true}
   /SetVariables(Particles.SetSocket:NiagaraInt32=&MakeCustomIntFromBool(
     Bool=&SetBoolByIntComparison(A=$Particles.PartitionParticles.Partition, B=0, Comparison Type=::NewEnumerator0)
     True INT=$Particles.Socket_1
     False INT=$Particles.Socket001
    )) {isSetParameters=true}
   &SkeletalMeshLocation(
    Bone / Socket Index=$Particles.SetSocket
    MaskSpawnGroup=false
    Mesh Sampling Type=::NewEnumerator5
    Orientation Sampling=::NewEnumerator1
    Position Sampling=::NewEnumerator1
    Sampled Position Offset=&AddVectorToPosition(
     Position=$Particles.ShapeLocation001.ShapePosition
     Vector=&SelectVectorFromArray(Direct Array Index=$Particles.SetSocket, Vector Selection Array=DataInterface() {propertyValues {FloatData [<0 1.1 25>, <0 0 2>]}})
    )
    Sampled Position Offset Coordinate Space:ENiagaraSkelMeshTransforms=::NewEnumerator2
    Skeletal Mesh=$System.Skeletal Mesh
    Socket Sampling Mode=::NewEnumerator5
   ) {module=SkeletalMeshLocation002}
   &ScaleSpriteSize(Uniform Curve Sprite Scale=DataInterface() {propertyValues {Curve {keys [<at 0 0 RCIM_Cubic>, {arriveTangent=17.0125, interpMode=RCIM_Cubic, leaveTangent=17.0125, tangentMode=RCTM_User, time=0.1, value=1}]}}})
   &GravityForce(Gravity=<0 0 -500>)
   &AerodynamicDrag(Aerodynamic Drag=0.1, Use Linear Drag=true, Use Rotational Drag=true)
   &SolveForcesAndVelocity(Clamp Velocity=false, Limit Acceleration=false)
   &ScaleSpriteSizeBySpeed(Max Scale Factor=<1 3>, Velocity Threshold=2000)
   &AlignSpriteToMeshOrientation(SetAlignment=true, SetFacing=true)
   &ScaleColor(
    Scale Alpha=&FloatFromCurve(FloatCurve=DataInterface() {propertyValues {Curve {keys [
         {arriveTangentWeight=0.1124, interpMode=RCIM_Cubic, leaveTangentWeight=0.2247, tangentMode=RCTM_User, time=0.1, value=0}
         {
          arriveTangent=0.0435
          arriveTangentWeight=0.0333
          interpMode=RCIM_Cubic
          leaveTangent=0.0435
          leaveTangentWeight=0.1
          tangentMode=RCTM_User
          time=0.3
          value=1
         }
         {
          arriveTangent=0.0176
          arriveTangentWeight=0.1
          interpMode=RCIM_Cubic
          leaveTangent=0.0176
          leaveTangentWeight=0.1334
          tangentMode=RCTM_User
          time=0.6
          value=1
         }
         <at 1 0 RCIM_Cubic>
        ]}}})
    ScaleA=true
    ScaleRGB=true
   )
  ]
  rendererClasses [&NiagaraSpriteRendererProperties]
  renderers [{rendererClass=&NiagaraSpriteRendererProperties}]
 }
 HydroBeam_Bass_Liquid {
  emitterUpdateScript [
   &EmitterState(
    Inactive Response=::NewEnumerator1
    Life Cycle Mode=::NewEnumerator1
    Loop Behavior=::NewEnumerator1
    Loop Duration Mode=::NewEnumerator1
    UseLoopDelay=false
   )
   &SpawnBurst_Instantaneous(Use Spawn Probability=false, UseLoopCountLimit=false)
   /SetVariables(Emitter.Orientation:Quat4f=$Engine.Owner.Rotation) {isSetParameters=true, module=SetVariables_#4}
   /SetVariables(Emitter.Position:Vector3f=/Niagara/DynamicInputs/Helpers/SimulationPosition()) {isSetParameters=true, module=SetVariables_#5}
   /SetVariables(
    Emitter.RotationVelocity:Vector3f=&Divide_Vector(
     A:Vector3f=&Subtract_Vector(
      A=&RotateVectorByQuaternion(Vector To Rotate=<0 0 1>, Quaternion=$Emitter.Orientation)
      B=&RotateVectorByQuaternion(Vector To Rotate=<0 0 1>, Quaternion=$Emitter.Previous.Orientation)
     )
     B:Vector3f=&VectorFromFloat(Value=$Engine.DeltaTime)
    )
    Emitter.Velocity:Vector3f=&Divide_Vector(
     A:Vector3f=&Subtract_Vector(A=$Emitter.Position, B=$Emitter.Previous.Position)
     B:Vector3f=&VectorFromFloat(Value=$Engine.DeltaTime)
    )
    Emitter.Viscosity=0.25
   ) {isSetParameters=true, module=SetVariables_#6}
   /SetVariables(Emitter.InheritSourceMovement:NiagaraFloat=&Lerp_Float(A=0, B=0.9, Alpha=$Emitter.Viscosity)) {isSetParameters=true, module=SetVariables_#7}
   /SetVariables(Emitter.EmissiveMultiplier:NiagaraFloat=&MakeCustomFloatFromBool(Bool=$User.bIsFrontEndPreview, True Float=15, False Float=1.25)) {isSetParameters=true, module=SetVariables_#8}
  ]
  particleSpawnScript [
   &InitializeParticle(
    Color Mode=::NewEnumerator0
    Mesh Scale Mode=::NewEnumerator3
    Position Offset=&Multiply_VectorByFloat(Vector=<0 0 -1>, Float=16)
    UsePositionOffset=true
    Write Material Random=true
   )
   &SkeletalMeshLocation(
    Bone / Socket Sampling Mode=::NewEnumerator5
    MaskSpawnGroup=false
    Mesh Sampling Type=::NewEnumerator6
    Orientation Sampling=::NewEnumerator1
    Position Sampling=::NewEnumerator2
    Sampled Position Offset=&SelectVectorFromArray(Direct Array Index=0, Vector Selection Array=DataInterface() {propertyValues {FloatData [<0 0 0>]}})
    Sampled Position Offset Coordinate Space:ENiagaraSkelMeshTransforms=::NewEnumerator1
    Skeletal Mesh=$System.Skeletal Mesh
   )
  ]
  particleUpdateScript [
   &ParticleState(Kill Particles When Lifetime Has Elapsed=false, Let Infinitely Lived Particles Die When Emitter Deactivates=true)
   &ScaleMeshSize(Scale Factor=&VectorFromFloat(Value=$System.LiquidScaleRemap))
   &SkeletalMeshLocation(
    Bone / Socket Sampling Mode=::NewEnumerator5
    MaskSpawnGroup=false
    Mesh Sampling Type=::NewEnumerator6
    Orientation Sampling=::NewEnumerator1
    Position Sampling=::NewEnumerator2
    Sampled Position Offset=&SelectVectorFromArray(Direct Array Index=0, Vector Selection Array=DataInterface() {propertyValues {FloatData [<0 0 0>]}})
    Sampled Position Offset Coordinate Space:ENiagaraSkelMeshTransforms=::NewEnumerator1
    Skeletal Mesh=$System.Skeletal Mesh
   ) {module=SkeletalMeshLocation001}
   &SkeletalMeshLocation(
    Bone / Socket Index=1
    MaskSpawnGroup=false
    Mesh Sampling Type=::NewEnumerator5
    Orientation Sampling=::NewEnumerator1
    Position Sampling=::NewEnumerator2
    Skeletal Mesh=$System.Skeletal Mesh
    Socket Sampling Mode=::NewEnumerator5
   ) {module=SkeletalMeshLocation002}
   /SetVariables(Particles.Position_Origin:NiagaraPosition=$Engine.Emitter.SimulationPosition) {isSetParameters=true, module=SetVariables_#9}
   /SetVariables(Particles.Position_offset:NiagaraPosition=$Output.SkeletalMeshLocation002.SampledPositionWithoutOffset) {isSetParameters=true}
   &PendulumSetup(
    Module.Enable Spring Driver=true
    Pendulum Length:NiagaraFloat=&Multiply_Float(
     A=&Lerp_Float(A=10, B=10, Alpha=&MakeCustomFloatFromBool(Bool=$User.bIsFrontEndPreview, True Float=1, False Float=0))
     B=$System.LiquidScaleRemap
    )
    Pendulum Pivot:Vector3f=/Niagara/DynamicInputs/TypeConversions/MakeCustomPositionFromBool(
     Bool:NiagaraBool=$User.bIsFrontEndPreview
     True Position:NiagaraPosition=$Particles.Position_offset
     False Position:NiagaraPosition=$Particles.Position_Origin
    )
    Pendulum Rest Axis:Vector3f=&NormalizeVector(
     Vector To Normalize=&Add_Vector(
      A=<0 0 -1>
      B=&Lerp_Vector(
       A=<0 0 0>
       B=&TransformVector(Vector=<0 0 -1>, Destination Space=::Simulation)
       Alpha=$System.Timeline.Float
      )
     )
     Safe Normalize=true
     Safe Fallback Threshold=0
     Safe Fallback Vector=<0 0 0.001>
    )
    "Spring Driven Constraint (Tightness)"=1
   )
   &GravityForce(Gravity=&Lerp_Vector(A=<0 0 -980>, B=<0 0 -19.6>, Alpha=$Emitter.Viscosity))
   &Drag(Drag=&Lerp_Float(A=0, B=5, Alpha=$Emitter.Viscosity), Use Linear Drag=true, Use Rotational Drag=false)
   &InheritSourceMovement(
    Applied Position Scale=&VectorFromFloat(Value=$Emitter.InheritSourceMovement)
    ApplyForce=false
    ApplyPosition=true
    ApplyPositionFalloff=false
    Source Velocity:Vector3f=$Emitter.Velocity
   )
   &SolveForcesAndVelocity(Clamp Velocity=false, Limit Acceleration=false, Speed Limit=200)
   &PendulumConstraint()
   &CalculateAccurateVelocity(Previous Position:NiagaraPosition=&Add_Vector(
     A=Linked() {linkedVariable {name=Particles.Previous.Position, type=NiagaraPosition}}
     B=&Multiply_VectorByFloat(Vector=$Emitter.Velocity, Float=&Multiply_Float(A=$Emitter.InheritSourceMovement, B=$Engine.DeltaTime))
    ))
   /Niagara/Modules/Update/Orientation/V2/OrientMeshToVector(
    Look At Direction:Vector3f=&NormalizeVector(
     Vector To Normalize=&Subtract_Vector(
      A=&MakeCustomVectorFromBool(
       Bool=$User.bIsFrontEndPreview
       True Vector=Linked() {linkedVariable {name=Particles.Position_offset, type=NiagaraPosition}}
       False Vector=Linked() {linkedVariable {name=Particles.Position_Origin, type=NiagaraPosition}}
      )
      B=Linked() {linkedVariable {name=Particles.Position, type=NiagaraPosition}}
     )
     Safe Normalize=true
     Safe Fallback Threshold=0
     Safe Fallback Vector=<0 0 0.001>
    )
    Override Quat=false
    Reference Axis Vector:Vector3f=<0 0 1>
    UseOptionalSideAxis=false
    UseOptionalUpAxis=false
   )
   &InterpolateOverTime(
    Rate Of Change=2
    Target Float Value:NiagaraFloat=&Max_Float(
     A:NiagaraFloat=&RemapRange(
      Input Value To Remap=&DotProduct(
       Vector A=<0 0 -1>
       Vector B=&ReturnMeshOrientationAxis(Mesh Orientation Axis=::ZAxis, Mesh Orientation Quat=$Particles.MeshOrientation)
      )
      Input Min=-1
      Input Max=0
      Output Min=0
      Output Max=1
     )
     B:NiagaraFloat=&RemapRange(
      Input Value To Remap=&Divide_Float(
       A=&Subtract_Float(A=$Particles.DistanceTraveled, B=$Particles.Previous.DistanceTraveled)
       B=$System.LiquidScaleRemap
      )
      Input Min=2
      Input Max=4
      Output Min=0
      Output Max=1
     )
    )
   )
   /SetVariables(Particles.LiquidSpeed:NiagaraFloat=&ClampFloat(Float=$Particles.InterpolateOverTime.InterpolateOverTime_Float.Moving Average)) {isSetParameters=true, module=SetVariables_#10}
   &Timeline(Play=&SetBoolByFloatComparison(A=$Particles.LiquidSpeed, B=0.333), Play Rate=3, Timeline End=10)
   /SetVariables(Particles.Turbulence:NiagaraFloat=$Particles.Timeline.NormalizedTime) {isSetParameters=true, module=SetVariables_#11}
   /Niagara/Modules/Update/Utility/IncrementOverTime(Rate of Change:NiagaraFloat=&Lerp_Float(A=1, B=1.25, Alpha=$Particles.LiquidSpeed))
   /SetVariables(
    Particles.DynamicMaterialParameter:Vector4f=&MakeFloat4FromVectorAndFloat(
     "Vector (RGB)"=&ReturnMeshOrientationAxis(Mesh Orientation Axis=::ZAxis, Mesh Orientation Quat=$Particles.MeshOrientation)
     "Float (Alpha)"=&Lerp_Float(A=0, B=0.6, Alpha=$System.Timeline.NormalizedTime)
    )
    Particles.DynamicMaterialParameter1:Vector4f=&MakeFloat4FromVectorAndFloat(
     "Vector (RGB)"=&ReturnMeshOrientationAxis(Mesh Orientation Axis=::XAxis, Mesh Orientation Quat=$Particles.MeshOrientation)
     "Float (Alpha)"=$Particles.IncrementOverTime.Counter
    )
    Particles.DynamicMaterialParameter2:Vector4f=&MakeVector4(X=$Particles.LiquidSpeed, Y=$Particles.Turbulence, Z=0, W=0)
   ) {isSetParameters=true, module=SetVariables_#12}
   &CameraOffset()
  ]
  rendererClasses [&NiagaraMeshRendererProperties]
  renderers [
   {rendererClass=&NiagaraMeshRendererProperties, rendererIndex=0}
   {rendererClass=&NiagaraMeshRendererProperties, rendererIndex=1}
  ]
 }
 HydroBeam_Bassr_Liquid {
  bEnabled=false
  emitterUpdateScript [
   &EmitterState(
    Inactive Response=::NewEnumerator1
    Life Cycle Mode=::NewEnumerator1
    Loop Behavior=::NewEnumerator1
    Loop Duration Mode=::NewEnumerator1
    UseLoopDelay=false
   )
   &SpawnBurst_Instantaneous(Use Spawn Probability=false, UseLoopCountLimit=false)
   /SetVariables(Emitter.Orientation:Quat4f=$Engine.Owner.Rotation) {isSetParameters=true, module=SetVariables_#4}
   /SetVariables(Emitter.Position:Vector3f=&SimulationPosition()) {isSetParameters=true, module=SetVariables_#5}
   /SetVariables(
    Emitter.RotationVelocity:Vector3f=&Divide_Vector(
     A:Vector3f=&Subtract_Vector(
      A=&RotateVectorByQuaternion(Vector To Rotate=<0 0 1>, Quaternion=$Emitter.Orientation)
      B=&RotateVectorByQuaternion(Vector To Rotate=<0 0 1>, Quaternion=$Emitter.Previous.Orientation)
     )
     B:Vector3f=&VectorFromFloat(Value=$Engine.DeltaTime)
    )
    Emitter.Velocity:Vector3f=&Divide_Vector(
     A:Vector3f=&Subtract_Vector(A=$Emitter.Position, B=$Emitter.Previous.Position)
     B:Vector3f=&VectorFromFloat(Value=$Engine.DeltaTime)
    )
    Emitter.Viscosity=0.25
   ) {isSetParameters=true, module=SetVariables_#6}
   /SetVariables(Emitter.InheritSourceMovement:NiagaraFloat=&Lerp_Float(A=0, B=0.9, Alpha=$Emitter.Viscosity)) {isSetParameters=true, module=SetVariables_#7}
   /SetVariables(Emitter.EmissiveMultiplier:NiagaraFloat=&MakeCustomFloatFromBool(Bool=$User.bIsFrontEndPreview, True Float=15, False Float=1.25)) {isSetParameters=true, module=SetVariables_#8}
  ]
  particleSpawnScript [
   &InitializeParticle(
    Color Mode=::NewEnumerator0
    Mesh Scale Mode=::NewEnumerator3
    Mesh Uniform Scale=$System.LiquidScaleRemap
    Position Offset=&Multiply_VectorByFloat(Vector=<0 0 -1>, Float=16)
    UsePositionOffset=true
    Write Material Random=true
   )
   &SkeletalMeshLocation(
    Bone / Socket Sampling Mode=::NewEnumerator5
    MaskSpawnGroup=false
    Mesh Sampling Type=::NewEnumerator6
    Orientation Sampling=::NewEnumerator1
    Position Sampling=::NewEnumerator2
    Sampled Position Offset=&SelectVectorFromArray(Direct Array Index=0, Vector Selection Array=DataInterface() {propertyValues {FloatData [<0 0 0>]}})
    Sampled Position Offset Coordinate Space:ENiagaraSkelMeshTransforms=::NewEnumerator1
    Skeletal Mesh=$System.Skeletal Mesh
   )
  ]
  particleUpdateScript [
   &ParticleState(Kill Particles When Lifetime Has Elapsed=false, Let Infinitely Lived Particles Die When Emitter Deactivates=true)
   &SkeletalMeshLocation(
    Bone / Socket Sampling Mode=::NewEnumerator5
    MaskSpawnGroup=false
    Mesh Sampling Type=::NewEnumerator6
    Orientation Sampling=::NewEnumerator1
    Position Sampling=::NewEnumerator2
    Sampled Position Offset=&SelectVectorFromArray(Direct Array Index=0, Vector Selection Array=DataInterface() {propertyValues {FloatData [<0 0 0>]}})
    Sampled Position Offset Coordinate Space:ENiagaraSkelMeshTransforms=::NewEnumerator1
    Skeletal Mesh=$System.Skeletal Mesh
   ) {module=SkeletalMeshLocation001}
   &ScaleMeshSize(Scale Factor=&VectorFromFloat(Value=0.5)) {enabled=false}
   /SetVariables(Particles.Position_Origin:NiagaraPosition=$Engine.Emitter.SimulationPosition) {isSetParameters=true, module=SetVariables_#9}
   /SetVariables(Particles.Position_Origin:NiagaraPosition=&AddVectorToPosition(Position=$Particles.SkeletalMeshLocation001.SampledPosition, Vector=<0 0 16>)) {enabled=false, isSetParameters=true, module=SetVariables_#13}
   &PendulumSetup(
    Module.Enable Spring Driver=true
    Pendulum Length:NiagaraFloat=&Multiply_Float(
     A=&Lerp_Float(A=10, B=10, Alpha=&MakeCustomFloatFromBool(Bool=$User.bIsFrontEndPreview, True Float=1, False Float=0))
     B=$System.LiquidScaleRemap
    )
    Pendulum Pivot:NiagaraPosition=$Particles.Position_Origin
    Pendulum Rest Axis:Vector3f=&NormalizeVector(
     Vector To Normalize=&Add_Vector(
      A=<0 0 -1>
      B=&Lerp_Vector(
       A=<0 0 0>
       B=&TransformVector(Vector=<0 0 -1>, Destination Space=::Simulation)
       Alpha=$System.Timeline.Float
      )
     )
     Safe Normalize=true
     Safe Fallback Threshold=0
     Safe Fallback Vector=<0 0 0.001>
    )
    "Spring Driven Constraint (Tightness)"=1
   )
   &GravityForce(Gravity=&Lerp_Vector(A=<0 0 -980>, B=<0 0 -19.6>, Alpha=$Emitter.Viscosity))
   &Drag(Drag=&Lerp_Float(A=0, B=5, Alpha=$Emitter.Viscosity), Use Linear Drag=true, Use Rotational Drag=false)
   &InheritSourceMovement(
    Applied Position Scale=&VectorFromFloat(Value=$Emitter.InheritSourceMovement)
    ApplyForce=false
    ApplyPosition=true
    ApplyPositionFalloff=false
    Source Velocity:Vector3f=$Emitter.Velocity
   )
   &SolveForcesAndVelocity(Clamp Velocity=false, Limit Acceleration=false, Speed Limit=200)
   &PendulumConstraint()
   &CalculateAccurateVelocity(Previous Position:NiagaraPosition=&Add_Vector(
     A=Linked() {linkedVariable {name=Particles.Previous.Position, type=NiagaraPosition}}
     B=&Multiply_VectorByFloat(Vector=$Emitter.Velocity, Float=&Multiply_Float(A=$Emitter.InheritSourceMovement, B=$Engine.DeltaTime))
    ))
   &OrientMeshToVector(
    Look At Direction:Vector3f=&NormalizeVector(
     Vector To Normalize=&Subtract_Vector(
      A=Linked() {linkedVariable {name=Particles.Position_Origin, type=NiagaraPosition}}
      B=Linked() {linkedVariable {name=Particles.Position, type=NiagaraPosition}}
     )
     Safe Normalize=true
     Safe Fallback Threshold=0
     Safe Fallback Vector=<0 0 0.001>
    )
    Override Quat=false
    Reference Axis Vector:Vector3f=<0 0 1>
    UseOptionalSideAxis=false
    UseOptionalUpAxis=false
   )
   &InterpolateOverTime(
    Rate Of Change=2
    Target Float Value:NiagaraFloat=&Max_Float(
     A:NiagaraFloat=&RemapRange(
      Input Value To Remap=&DotProduct(
       Vector A=<0 0 -1>
       Vector B=&ReturnMeshOrientationAxis(Mesh Orientation Axis=::ZAxis, Mesh Orientation Quat=$Particles.MeshOrientation)
      )
      Input Min=-1
      Input Max=0
      Output Min=0
      Output Max=1
     )
     B:NiagaraFloat=&RemapRange(
      Input Value To Remap=&Divide_Float(
       A=&Subtract_Float(A=$Particles.DistanceTraveled, B=$Particles.Previous.DistanceTraveled)
       B=$System.LiquidScaleRemap
      )
      Input Min=2
      Input Max=4
      Output Min=0
      Output Max=1
     )
    )
   )
   /SetVariables(Particles.LiquidSpeed:NiagaraFloat=&ClampFloat(Float=$Particles.InterpolateOverTime.InterpolateOverTime_Float.Moving Average)) {isSetParameters=true, module=SetVariables_#10}
   &Timeline(Play=&SetBoolByFloatComparison(A=$Particles.LiquidSpeed, B=0.333), Play Rate=3, Timeline End=10)
   /SetVariables(Particles.Turbulence:NiagaraFloat=$Particles.Timeline.NormalizedTime) {isSetParameters=true, module=SetVariables_#11}
   &IncrementOverTime(Rate of Change:NiagaraFloat=&Lerp_Float(A=1, B=1.25, Alpha=$Particles.LiquidSpeed))
   /SetVariables(
    Particles.DynamicMaterialParameter:Vector4f=&MakeFloat4FromVectorAndFloat(
     "Vector (RGB)"=&ReturnMeshOrientationAxis(Mesh Orientation Axis=::ZAxis, Mesh Orientation Quat=$Particles.MeshOrientation)
     "Float (Alpha)"=&Lerp_Float(A=0, B=0.6, Alpha=$System.Timeline.NormalizedTime)
    )
    Particles.DynamicMaterialParameter1:Vector4f=&MakeFloat4FromVectorAndFloat(
     "Vector (RGB)"=&ReturnMeshOrientationAxis(Mesh Orientation Axis=::XAxis, Mesh Orientation Quat=$Particles.MeshOrientation)
     "Float (Alpha)"=$Particles.IncrementOverTime.Counter
    )
    Particles.DynamicMaterialParameter2:Vector4f=&MakeVector4(X=$Particles.LiquidSpeed, Y=$Particles.Turbulence, Z=0, W=0)
   ) {isSetParameters=true, module=SetVariables_#12}
   &CameraOffset()
  ]
  rendererClasses [&NiagaraMeshRendererProperties]
  renderers [
   {rendererClass=&NiagaraMeshRendererProperties, rendererIndex=0}
   {rendererClass=&NiagaraMeshRendererProperties, rendererIndex=1}
  ]
 }
 Water_Meter {
  emitterUpdateScript [
   &EmitterState(
    Inactive Response=::NewEnumerator1
    Life Cycle Mode=::NewEnumerator1
    Loop Behavior=::NewEnumerator1
    Loop Duration Mode=::NewEnumerator1
    UseLoopDelay=false
   )
   &SpawnBurst_Instantaneous(Use Spawn Probability=false, UseLoopCountLimit=false)
   /SetVariables(Emitter.Orientation:Quat4f=$Engine.Owner.Rotation) {isSetParameters=true, module=SetVariables_#4}
   /SetVariables(Emitter.Position:Vector3f=&SimulationPosition()) {isSetParameters=true, module=SetVariables_#5}
   /SetVariables(
    Emitter.RotationVelocity:Vector3f=&Divide_Vector(
     A:Vector3f=&Subtract_Vector(
      A=&RotateVectorByQuaternion(Vector To Rotate=<0 0 1>, Quaternion=$Emitter.Orientation)
      B=&RotateVectorByQuaternion(Vector To Rotate=<0 0 1>, Quaternion=$Emitter.Previous.Orientation)
     )
     B:Vector3f=&VectorFromFloat(Value=$Engine.DeltaTime)
    )
    Emitter.Velocity:Vector3f=&Divide_Vector(
     A:Vector3f=&Subtract_Vector(A=$Emitter.Position, B=$Emitter.Previous.Position)
     B:Vector3f=&VectorFromFloat(Value=$Engine.DeltaTime)
    )
    Emitter.Viscosity=1
   ) {isSetParameters=true, module=SetVariables_#6}
   /SetVariables(Emitter.InheritSourceMovement:NiagaraFloat=&Lerp_Float(A=0, B=0.9, Alpha=$Emitter.Viscosity)) {isSetParameters=true, module=SetVariables_#7}
   /SetVariables(Emitter.EmissiveMultiplier:NiagaraFloat=&MakeCustomFloatFromBool(Bool=$User.bIsFrontEndPreview, True Float=1.5, False Float=0.7)) {isSetParameters=true}
  ]
  particleSpawnScript [
   &InitializeParticle(
    Color Mode=::NewEnumerator0
    Mesh Scale Mode=::NewEnumerator3
    Position Offset=&Multiply_VectorByFloat(Vector=<0 0 -1>, Float=16)
    UsePositionOffset=true
    Write Material Random=true
   )
   &SkeletalMeshLocation(
    Bone / Socket Sampling Mode=::NewEnumerator5
    MaskSpawnGroup=false
    Mesh Sampling Type=::NewEnumerator6
    Orientation Sampling=::NewEnumerator1
    Position Sampling=::NewEnumerator2
    Sampled Position Offset=&SelectVectorFromArray(Direct Array Index=0, Vector Selection Array=DataInterface() {propertyValues {FloatData [<0 0 0>]}})
    Sampled Position Offset Coordinate Space:ENiagaraSkelMeshTransforms=::NewEnumerator1
    Skeletal Mesh=$System.Skeletal Mesh
   )
  ]
  particleUpdateScript [
   &ParticleState(Kill Particles When Lifetime Has Elapsed=false, Let Infinitely Lived Particles Die When Emitter Deactivates=true)
   &ScaleMeshSize(Scale Factor=&VectorFromFloat(Value=$System.LiquidScaleRemap))
   &SkeletalMeshLocation(
    Bone / Socket Sampling Mode=::NewEnumerator5
    MaskSpawnGroup=false
    Mesh Sampling Type=::NewEnumerator6
    Orientation Sampling=::NewEnumerator1
    Position Sampling=::NewEnumerator2
    Sampled Position Offset=&SelectVectorFromArray(Direct Array Index=0, Vector Selection Array=DataInterface() {propertyValues {FloatData [<0 0 0>]}})
    Sampled Position Offset Coordinate Space:ENiagaraSkelMeshTransforms=::NewEnumerator1
    Skeletal Mesh=$System.Skeletal Mesh
   ) {module=SkeletalMeshLocation001}
   /SetVariables(Particles.Position_Origin:NiagaraPosition=$Engine.Emitter.SimulationPosition) {isSetParameters=true, module=SetVariables_#9}
   /SetVariables(Particles.Position_Origin:NiagaraPosition=&AddVectorToPosition(Position=$Particles.SkeletalMeshLocation001.SampledPosition, Vector=<0 0 16>)) {enabled=false, isSetParameters=true, module=SetVariables_#13}
   &PendulumSetup(
    Module.Enable Spring Driver=true
    Pendulum Length:NiagaraFloat=&Multiply_Float(
     A=&Lerp_Float(A=10, B=10, Alpha=&MakeCustomFloatFromBool(Bool=$User.bIsFrontEndPreview, True Float=1, False Float=0))
     B=$System.LiquidScaleRemap
    )
    Pendulum Pivot:NiagaraPosition=$Particles.Position_Origin
    Pendulum Rest Axis:Vector3f=&NormalizeVector(
     Vector To Normalize=&Add_Vector(
      A=<0 0 -1>
      B=&Lerp_Vector(
       A=<0 0 0>
       B=&TransformVector(Vector=<0 0 -1>, Destination Space=::Simulation)
       Alpha=$System.Timeline.Float
      )
     )
     Safe Normalize=true
     Safe Fallback Threshold=0
     Safe Fallback Vector=<0 0 0.001>
    )
    "Spring Driven Constraint (Tightness)"=1
   )
   &GravityForce(Gravity=&Lerp_Vector(A=<0 0 -980>, B=<0 0 -19.6>, Alpha=$Emitter.Viscosity))
   &Drag(Drag=&Lerp_Float(A=0, B=5, Alpha=$Emitter.Viscosity), Use Linear Drag=true, Use Rotational Drag=false)
   &InheritSourceMovement(
    Applied Position Scale=&VectorFromFloat(Value=$Emitter.InheritSourceMovement)
    ApplyForce=false
    ApplyPosition=true
    ApplyPositionFalloff=false
    Source Velocity:Vector3f=$Emitter.Velocity
   )
   &SolveForcesAndVelocity(Clamp Velocity=false, Limit Acceleration=false, Speed Limit=200)
   &PendulumConstraint()
   &CalculateAccurateVelocity(Previous Position:NiagaraPosition=&Add_Vector(
     A=Linked() {linkedVariable {name=Particles.Previous.Position, type=NiagaraPosition}}
     B=&Multiply_VectorByFloat(Vector=$Emitter.Velocity, Float=&Multiply_Float(A=$Emitter.InheritSourceMovement, B=$Engine.DeltaTime))
    ))
   &OrientMeshToVector(
    Look At Direction:Vector3f=&NormalizeVector(
     Vector To Normalize=&Subtract_Vector(
      A=Linked() {linkedVariable {name=Particles.Position_Origin, type=NiagaraPosition}}
      B=Linked() {linkedVariable {name=Particles.Position, type=NiagaraPosition}}
     )
     Safe Normalize=true
     Safe Fallback Threshold=0
     Safe Fallback Vector=<0 0 0.001>
    )
    Override Quat=false
    Reference Axis Vector:Vector3f=<0 0 1>
    UseOptionalSideAxis=false
    UseOptionalUpAxis=false
   )
   &InterpolateOverTime(
    Rate Of Change=2
    Target Float Value:NiagaraFloat=&Max_Float(
     A:NiagaraFloat=&RemapRange(
      Input Value To Remap=&DotProduct(
       Vector A=<0 0 -1>
       Vector B=&ReturnMeshOrientationAxis(Mesh Orientation Axis=::ZAxis, Mesh Orientation Quat=$Particles.MeshOrientation)
      )
      Input Min=-1
      Input Max=0
      Output Min=0
      Output Max=1
     )
     B:NiagaraFloat=&RemapRange(
      Input Value To Remap=&Divide_Float(
       A=&Subtract_Float(A=$Particles.DistanceTraveled, B=$Particles.Previous.DistanceTraveled)
       B=$System.LiquidScaleRemap
      )
      Input Min=2
      Input Max=4
      Output Min=0
      Output Max=1
     )
    )
   )
   /SetVariables(Particles.LiquidSpeed:NiagaraFloat=&ClampFloat(Float=$Particles.InterpolateOverTime.InterpolateOverTime_Float.Moving Average)) {isSetParameters=true, module=SetVariables_#10}
   &Timeline(Play=&SetBoolByFloatComparison(A=$Particles.LiquidSpeed, B=0.333), Play Rate=3, Timeline End=10)
   /SetVariables(Particles.Turbulence:NiagaraFloat=$Particles.Timeline.NormalizedTime) {isSetParameters=true, module=SetVariables_#11}
   &IncrementOverTime(Rate of Change:NiagaraFloat=&Lerp_Float(A=1, B=1.25, Alpha=$Particles.LiquidSpeed))
   /SetVariables(
    Particles.DynamicMaterialParameter:Vector4f=&MakeFloat4FromVectorAndFloat(
     "Vector (RGB)"=&ReturnMeshOrientationAxis(Mesh Orientation Axis=::ZAxis, Mesh Orientation Quat=$Particles.MeshOrientation)
     "Float (Alpha)"=&Lerp_Float(A=0, B=0.55, Alpha=$System.Timeline.NormalizedTime)
    )
    Particles.DynamicMaterialParameter1:Vector4f=&MakeFloat4FromVectorAndFloat(
     "Vector (RGB)"=&ReturnMeshOrientationAxis(Mesh Orientation Axis=::XAxis, Mesh Orientation Quat=$Particles.MeshOrientation)
     "Float (Alpha)"=$Particles.IncrementOverTime.Counter
    )
    Particles.DynamicMaterialParameter2:Vector4f=&MakeVector4(X=$Particles.LiquidSpeed, Y=$Particles.Turbulence, Z=0, W=0)
   ) {isSetParameters=true, module=SetVariables_#12}
   &CameraOffset()
  ]
  rendererClasses [&NiagaraMeshRendererProperties]
  renderers [
   {rendererClass=&NiagaraMeshRendererProperties, rendererIndex=0}
   {rendererClass=&NiagaraMeshRendererProperties, rendererIndex=1}
  ]
 }
}
summary {
 systemName=NS_Bass_HydroBeam_Splash_Liquid1
 emitters [
  {bEnabled=false, emitterName=HydroBeam_Bassr_Liquid, rendererClasses [&NiagaraMeshRendererProperties]}
  {emitterName=Water_Meter, rendererClasses [&NiagaraMeshRendererProperties]}
  {emitterName=Dripping001, rendererClasses [&NiagaraSpriteRendererProperties]}
  {emitterName=HydroBeam_Bass_Liquid, rendererClasses [&NiagaraMeshRendererProperties]}
 ]
 userVariables [
  User.bIsFrontEndPreview {type=NiagaraBool}
  User.dripsocket1 {type=NiagaraFloat, defaultValue=0, description=""}
  User.dripsocket2 {type=NiagaraFloat, defaultValue=1, description=""}
  User.Duration {type=NiagaraFloat, defaultValue=0.15, description=""}
  User.IsFilling {type=NiagaraBool, defaultValue=true}
  User.IsFinished {type=NiagaraBool}
 ]
}
